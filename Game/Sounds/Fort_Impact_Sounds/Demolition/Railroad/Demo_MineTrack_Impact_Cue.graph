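# Sound cue for impacts on mine track during demolition, playing railroad rail impact sounds.
# A Random node picks one of three equally weighted rail impact waves and feeds it through a Modulator at the root for variation.
# It uses small melee weapon attenuation, the impacts sound class and a shared object-impact concurrency set to limit overlapping instances.
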SoundCue /Game/Sounds/Fort_Impact_Sounds/Demolition/Railroad/Demo_MineTrack_Impact_Cue
root=Modulator_37
nodes [
 {id=Modulator_37, children [Random_1]}
 {id=Random_1, children [WavePlayer_6,WavePlayer_7,WavePlayer_8], properties {weights [1,1,1]}}
 {id=WavePlayer_6, properties {soundWaveAssetPtr=&/railroad_rail_impact_01}}
 {id=WavePlayer_7, properties {soundWaveAssetPtr=&/railroad_rail_impact_02}}
 {id=WavePlayer_8, properties {soundWaveAssetPtr=&/railroad_rail_impact_03}}
]
properties {
 attenuationSettings=&Fort_Weapon_Melee_Sml_Att
 maxDistance=2800
 soundClassObject=&Fort_Impacts
 volumeMultiplier=0.63
 concurrencyOverrides {maxCount=3, resolutionRule=StopOldest}
 concurrencySet [&CONC_Object_Impacts_FourMax]
}
